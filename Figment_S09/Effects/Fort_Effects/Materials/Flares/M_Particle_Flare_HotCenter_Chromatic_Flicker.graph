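# Unlit additive particle sprite material for a flickering lens-style flare with a hot center and chromatic fringing.
# Powered sphere masks form the core and glow, while soft flare RGB samples separate by camera distance, tinted by parameter or particle color.
# Multiplied sines drive optional flicker, randomized for GPU particles or set by dynamic parameters, alongside depth and near-camera fades.
# World position offset combines curl-noise drift with distance-based sprite sizing.
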Material /Figment_S09/Effects/Fort_Effects/Materials/Flares/M_Particle_Flare_HotCenter_Chromatic_Flicker
graph {
 properties {bUsedWithParticleSprites=true, blendMode=BLEND_Additive, pixelDepthOffsetMode=PDOM_Legacy, shadingModel=MSM_Unlit, usedShadingModels=Unlit}
 nodes [
  Constant() {r=1}
  n1 = ParticleRandom()
  n2 = DynamicParameter() {paramNames [FlickerFade,FlickerSpeed,Param3,Param4]}
  n3 = Multiply(
   A=Time()
   B=StaticSwitchParameter(True=LinearInterpolate(Alpha=@n1) {constA=1.2, constB=0.8}, False=@n2.FlickerSpeed) {parameterName=GPU Particles}
  )
  n4 = Multiply(A=Sine(None=@n3) {period=0.155}, B=Sine(None=@n3) {period=0.2})
  n5 = Sine(None=@n3) {period=0.43}
  n6 = Clamp(None=Multiply(A=Sine(None=@n3) {period=0.075}, B=Multiply(A=@n4, B=@n5))) {maxDefault=3}
  n7 = ScalarParameter() {defaultValue=0.4, desc="0 = full flicker amount", parameterName=GPUFlickerFadeMin}
  n8 = ScalarParameter() {defaultValue=1, desc="1 = no flicker", parameterName=GPUFlickerFadeMax, sortPriority=1}
  n9 = ParticleRandom()
  n10 = LinearInterpolate(
   A=@n6
   Alpha=StaticSwitchParameter(True=LinearInterpolate(A=@n7, B=@n8, Alpha=@n9) {constA=0.5}, False=@n2.FlickerFade) {parameterName=GPU Particles}
  ) {constA=1}
  n11 = StaticSwitchParameter(True=FeatureLevelSwitch(Default=@n10, ES2=@n6, ES3_1=@n6, SM4_REMOVED=@n10), False=Constant() {r=1}) {parameterName=UseFlicker}
  n12 = ParticleColor()
  n13 = DepthFade(Opacity=@n12.A, FadeDistance=ScalarParameter() {defaultValue=30, parameterName=DepthFade})
  n14 = FeatureLevelSwitch(Default=@n13, ES2=@n12.A, ES3_1=@n12.A, SM4_REMOVED=@n13, SM5=@n13, SM6=@n13)
  n15 = Constant() {r=0.5}
  n16 = SphereMask(A=TextureCoordinate(), B=@n15, Radius=@n15) {hardnessPercent=0}
  n17 = Multiply(A=Power(Base=@n16) {constExponent=25}) {constB=2}
  n18 = VectorParameter() {parameterName=Color, defaultValue=<rgba 3.5 3.3425 1.881 0>}
  n19 = Multiply(A=Power(Base=@n16) {constExponent=4}) {constB=0.15}
  n20 = TextureCoordinate()
  n21 = MaterialFunctionCall(Distance=ScalarParameter() {defaultValue=2400, parameterName=SeparateChroma}) {materialFunction=&Near_Camera_Fade}
  n22 = Add(A=TextureCoordinate(), B=LinearInterpolate(Alpha=@n21.Result) {constA=0.03, constB=0.1}) {constB=0.03}
  n23 = TextureCoordinate()
  n24 = LinearInterpolate(Alpha=@n21.Result) {constA=-0.03, constB=-0.1}
  n25 = AppendVector(
   A=AppendVector(
    A=TextureSampleParameter2D(UVs=@n20) {parameterName=FlareMask, texture=/Game/Effects/Fort_Effects/Textures/Flares/T_Flare_Soft}.R
    B=TextureSampleParameter2D(UVs=@n22) {parameterName=FlareMask_1, texture=&T_Flare_Soft}.G
   )
   B=TextureSampleParameter2D(UVs=Add(A=@n23, B=@n24) {constB=-0.03}) {parameterName=FlareMask_2, texture=&T_Flare_Soft}.B
  )
  n26 = Add(A=Multiply(A=@n17, B=LinearInterpolate(B=@n18.RGB) {constA=1, constAlpha=0.75}), B=Multiply(A=Add(A=@n19, B=@n25), B=@n18.RGB))
  n27 = Multiply(A=@n26, B=ScalarParameter() {defaultValue=0.035, parameterName=Opacity})
  n28 = DepthFade(Opacity=@n27, FadeDistance=ScalarParameter() {defaultValue=15, parameterName=DepthFade_1})
  Multiply(
   A=StaticSwitchParameter(True=Multiply(A=@n11, B=@n14), False=FeatureLevelSwitch(Default=@n28, ES2=@n27, ES3_1=@n27, SM4_REMOVED=@n28)) {parameterName=UseParticleAlpha}
   B=MaterialFunctionCall(Distance=ScalarParameter() {defaultValue=100, parameterName=NearFadeDistance}) {materialFunction=&Near_Camera_Fade}.Result
  )
  n29 = VertexNormalWS()
  n30 = Time() {period=16}
  n31 = ScalarParameter() {defaultValue=0.11, parameterName=Slow}
  n32 = Subtract(A=Add(A=WorldPosition().XYZ, B=Multiply(A=@n30, B=@n31)), B=ActorPositionWS())
  n33 = Multiply(
   A=ScalarParameter() {defaultValue=25, parameterName=ScaleOffset}
   B=VectorNoise(World Position=Multiply(A=@n29, B=@n32)) {noiseFunction=VNF_CurlALU}
  )
  n34 = WorldPosition()
  n35 = LinearInterpolate(
   A=ScalarParameter() {defaultValue=40, parameterName=SpriteScaleNear}
   B=ScalarParameter() {defaultValue=30, parameterName=SpriteScaleFar}
   Alpha=MaterialFunctionCall(Distance=ScalarParameter() {defaultValue=500, parameterName=SoftenDistance}) {materialFunction=&Near_Camera_Fade}.Result
  )
  n36 = AppendVector(A=@n35, B=@n35)
  Reroute(None=Add(
    A=Transform(None=@n33)
    B=MaterialFunctionCall!(Center Location=@n34.XYZ, 0-1 UVs, XY scale=@n36, OverridePivot, "Normals (Optional)") {materialFunction=&Sprite}.World position offset
   ))
  n37 = LinearInterpolate(B=@n12.RGB) {constA=1, constAlpha=0.7}
  n38 = Add(A=@n19, B=@n25)
  StaticSwitchParameter(True=Add(A=Multiply(A=@n17, B=@n37), B=Multiply(A=@n38, B=@n12.RGB)), False=@n26) {parameterName=UseParticleColor}
  WorldPosition()
 ]
}
